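# Mobile water material function that outputs an animated, Z-only wave offset for the water surface.
# World XY position, by default, is scaled and scrolled over time to sample a grayscale wave texture, whose red channel is remapped into a signed height.
# A saturated, biased Water Depth mask multiplies that height, flattening waves in shallow water and reaching full strength as depth increases.
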MaterialFunction /FortWater/Materials/WaterSurface/Mobile/Functions/MF_MobileWaterWaves
graph {nodes [
  n1 = FunctionInput(Preview=WorldPosition().XYZ) {bUsePreviewValueAsDefault=true, inputName=Position}
  n2 = Subtract(A=Multiply(A=ComponentMask(None=@n1) {g=true, r=true}) {constB=0.0001}, B=Multiply(A=Time()) {constB=0.05})
  n3 = Add(A=Multiply(
    A=TextureSample(UVs=@n2) {samplerType=SAMPLERTYPE_LinearGrayscale, texture=/FortWater/Materials/WaterSurface/Mobile/Textures/T_MobileWater_Waves}.R
    B=Constant() {r=-45}
   )) {constB=15}
  n4 = FunctionInput() {bUsePreviewValueAsDefault=true, inputName=Water Depth, inputType=FunctionInput_Scalar, sortPriority=1}
  n5 = Saturate(None=Subtract(A=Multiply(A=@n4) {constB=0.004}) {constB=0.3})
  FunctionOutput(None=Multiply(A=Multiply(A=@n3, B=@n5), B=Constant3Vector() {constant=<rgba 0 0 1 0>}))
 ]}
